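# NPC conversation database for Season 26 Battle Pass Snapshot Part 02 story quests Q02 and Q11.
# Each entry tag starts a three-line speech chain: the PizzaParty NPC (Piper Pace) for Q02, the BlazerVeil NPC receiving a ring for Q11.
# Auto-selected main-questline choices require the quest active, not completed, and the matching NPC tag.
# Each final choice carries a quest tag and leads to a task completing that quest's obj0.
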ConversationDatabase /BattlePassS26_Quests/Conversations/Snapshots/Part02/Conversation_S26_Snapshot_P02
entry_points [
 {tag=Conversation.Entry.Season26.Snapshot.P02.Q02, destinations [#1]}
 {tag=Conversation.Entry.Season26.Snapshot.P02.Q11, destinations [#2]}
]
nodes [
 ConversationEntryPointNode_0 {
  id=#2
  class=ConversationEntryPointNode
  outputs [#3]
  properties {entryTag {tagName=Conversation.Entry.Season26.Snapshot.P02.Q11}}
 }
 Speech_C_0 {
  id=#3
  class=Speech_C
  outputs [#4]
  properties {softTaskIcon=&T_UI_NPC_MainQuest, generalConfig {defaultMessage="Nolan's very very sorry, is he?"}}
  sub_nodes [
   ConversationChoiceNode_0 {
    class=ConversationChoiceNode
    properties {
     defaultChoiceDisplayText=Offer Ring
     choiceTags {gameplayTags [{tagName=Conversation.Choice.AutoSelectIfValid}, {tagName=Conversation.Choice.Category.MainQuestline}]}
    }
   }
   FortConversationRequirement_ControllerRequirement_0 {
    class=FortConversationRequirement_ControllerRequirement
    properties {
     nodeName=Require: Has Completed Quest
     requirement {
      class=FortControllerRequirement_HasCompletedQuest
      properties {bInverseCondition=true, questDefinition=/BattlePassS26_Quests/Items/QuestItems/Snapshot/Part02/quest_s26_snapshot_p02_q11}
     }
    }
   }
   FortConversationRequirement_ControllerRequirement_1 {
    class=FortConversationRequirement_ControllerRequirement
    properties {
     nodeName=Require: Is Quest Active
     requirement {
      class=FortControllerRequirement_IsQuestActive
      properties {questDefinition=&quest_s26_snapshot_p02_q11}
     }
    }
   }
   FortConversationRequirement_ControllerRequirement_3 {
    class=FortConversationRequirement_ControllerRequirement
    properties {
     nodeName=Require: NPC Has Conversation Tag
     participantId {tagName=Conversation.Participant.NPC}
     requirement {class=FortControllerRequirement_NPCHasConversationEntryTag, properties {queryTag {tagName=Conversation.Entry.BlazerVeil}}}
    }
   }
  ]
 }
 Speech_C_3 {
  id=#4
  class=Speech_C
  outputs [#5]
  properties {generalConfig {defaultMessage=Ugh. I love it when he grovels. And his skills... It's hard to replace them.}}
  sub_nodes [ConversationChoiceNode_8 {class=ConversationChoiceNode, properties {defaultChoiceDisplayText=Continue}}]
 }
 Speech_C_4 {
  id=#5
  class=Speech_C
  outputs [#6]
  properties {generalConfig {defaultMessage="Tell him I'm in. But if he ditches me again, it's over for good."}}
  sub_nodes [ConversationChoiceNode_10 {
    class=ConversationChoiceNode
    properties {defaultChoiceDisplayText=Continue, choiceTags {gameplayTags [{tagName=Athena.Quests.S26.Conversations.P02.Choice.Q11}]}}
   }]
 }
 FortConversationTaskNode_CompleteQuestObjectiveByBackendName_0 {
  id=#6
  class=FortConversationTaskNode_CompleteQuestObjectiveByBackendName
  properties {
   objectiveBackendName=quest_s26_snapshot_p02_q11_obj0
   questDefinition=&quest_s26_snapshot_p02_q11
  }
 }
 ConversationEntryPointNode_1 {
  id=#1
  class=ConversationEntryPointNode
  outputs [#7]
  properties {entryTag {tagName=Conversation.Entry.Season26.Snapshot.P02.Q02}}
 }
 Speech_C_1 {
  id=#7
  class=Speech_C
  outputs [#8]
  properties {
   softTaskIcon=&T_UI_NPC_MainQuest
   generalConfig {defaultMessage="You're doing work for Nolan Chance, huh? Haven't heard that name in a minute."}
  }
  sub_nodes [
   ConversationChoiceNode_1 {
    class=ConversationChoiceNode
    properties {
     defaultChoiceDisplayText=Talk to Piper Pace
     choiceTags {gameplayTags [{tagName=Conversation.Choice.AutoSelectIfValid}, {tagName=Conversation.Choice.Category.MainQuestline}]}
    }
   }
   FortConversationRequirement_ControllerRequirement_2 {
    class=FortConversationRequirement_ControllerRequirement
    properties {
     nodeName=Require: Has Completed Quest
     requirement {
      class=FortControllerRequirement_HasCompletedQuest
      properties {bInverseCondition=true, questDefinition=/BattlePassS26_Quests/Items/QuestItems/Snapshot/Part02/quest_s26_snapshot_p02_q02}
     }
    }
   }
   FortConversationRequirement_ControllerRequirement_4 {
    class=FortConversationRequirement_ControllerRequirement
    properties {
     nodeName=Require: Is Quest Active
     requirement {
      class=FortControllerRequirement_IsQuestActive
      properties {questDefinition=&quest_s26_snapshot_p02_q02}
     }
    }
   }
   FortConversationRequirement_ControllerRequirement_5 {
    class=FortConversationRequirement_ControllerRequirement
    properties {
     nodeName=Require: NPC Has Conversation Tag
     participantId {tagName=Conversation.Participant.NPC}
     requirement {class=FortControllerRequirement_NPCHasConversationEntryTag, properties {queryTag {tagName=Conversation.Entry.PizzaParty}}}
    }
   }
  ]
 }
 Speech_C_2 {
  id=#8
  class=Speech_C
  outputs [#9]
  properties {generalConfig {defaultMessage=I've been saving up to chase my dream of being a pro circuit racer... but the tips aren't cutting it.}}
  sub_nodes [ConversationChoiceNode_2 {class=ConversationChoiceNode, properties {defaultChoiceDisplayText=Continue}}]
 }
 Speech_C_10 {
  id=#9
  class=Speech_C
  outputs [#10]
  properties {generalConfig {defaultMessage="If you can show me you're not going to be a liability on the job, I'm in. So impress me."}}
  sub_nodes [ConversationChoiceNode_3 {
    class=ConversationChoiceNode
    properties {defaultChoiceDisplayText=Continue, choiceTags {gameplayTags [{tagName=Athena.Quests.S26.Conversations.P02.Choice.Q02}]}}
   }]
 }
 FortConversationTaskNode_CompleteQuestObjectiveByBackendName_2 {
  id=#10
  class=FortConversationTaskNode_CompleteQuestObjectiveByBackendName
  properties {
   objectiveBackendName=quest_s26_snapshot_p02_q02_obj0
   questDefinition=&quest_s26_snapshot_p02_q02
  }
 }
]
